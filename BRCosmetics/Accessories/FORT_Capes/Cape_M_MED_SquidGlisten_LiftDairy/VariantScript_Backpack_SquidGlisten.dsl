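; Backpack variant script for the SquidGlistenLift cosmetic that picks which variant tag the back bling uses in a loadout.
; DetermineLoadoutVariantSelection returns Mat1 when the loadout contains the matching character, the backpack's Parts channel is Stage1 and the character's Progressive channel is Stage1.
; In all other cases it returns the backpack's active Parts variant unchanged.
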(fn DetermineLoadoutVariantSelection
  (params (Loadout :type Cosmetic%20Loadout%20Structure%20%28by%20ref%29))
  (if (Class|FortCosmeticFunctionLibrary|IsItemInLoadout Loadout "/BRCosmetics/Athena/Items/Cosmetics/Characters/Character_SquidGlistenLift.Character_SquidGlistenLift")
    (if (== (Class|FortCosmeticFunctionLibrary|GetActiveVariantForSoftItemAndChannel Loadout "/BRCosmetics/Athena/Items/Cosmetics/Backpacks/Backpack_SquidGlistenLift.Backpack_SquidGlistenLift" "(TagName=\"Cosmetics.Variant.Channel.Parts\")") "(TagName=\"Cosmetics.Variant.Property.Stage1\")")
      (if (== (Class|FortCosmeticFunctionLibrary|GetActiveVariantForSoftItemAndChannel Loadout "/BRCosmetics/Athena/Items/Cosmetics/Characters/Character_SquidGlistenLift.Character_SquidGlistenLift" "(TagName=\"Cosmetics.Variant.Channel.Progressive\")") "(TagName=\"Cosmetics.Variant.Property.Stage1\")")
        (return "(TagName=\"Cosmetics.Variant.Property.Mat1\")"))
      (else
        _)))
  (return (Class|FortCosmeticFunctionLibrary|GetActiveVariantForSoftItemAndChannel Loadout "/BRCosmetics/Athena/Items/Cosmetics/Backpacks/Backpack_SquidGlistenLift.Backpack_SquidGlistenLift" "(TagName=\"Cosmetics.Variant.Channel.Parts\")")))
